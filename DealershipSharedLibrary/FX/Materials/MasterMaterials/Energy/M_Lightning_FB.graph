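# Translucent unlit lightning material for Niagara sprites, ribbons and mesh particles.
# It blends two lightning texture samples, one V-mirrored, each with an optional fixed UV rotation, alongside dynamic parameters for detail motion, brightness, rotation and texture switch.
# Rotating normal-map noise supplies UV distortion, and panning plasma noise modulates a scrolling grid mask.
# Particle color drives boosted emissive, while particle alpha, vertical gradients and depth fade shape opacity.
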Material /DealershipSharedLibrary/FX/Materials/MasterMaterials/Energy/M_Lightning_FB
graph {
 properties {
  bAutomaticallySetUsageInEditor=false
  bUsedWithNiagaraMeshParticles=true
  bUsedWithNiagaraRibbons=true
  bUsedWithNiagaraSprites=true
  blendMode=BLEND_Translucent
  shadingModel=MSM_Unlit
  usedShadingModels=Unlit
 }
 nodes [
  MaterialFunctionCall(Emissive=NamedRerouteUsage()) {materialFunction=&MF_FX_EmissiveAdjustments}
  n1 = NamedRerouteUsage()
  n2 = Add(A=NamedRerouteUsage(), B=TextureCoordinate())
  n3 = Rotator(Coordinate=@n2, Time=Constant() {r=0.5}) {speed=3.14}
  n4 = NamedRerouteUsage()
  n5 = Multiply(
   A=TextureSampleParameter2D(UVs=LinearInterpolate(A=@n3, B=@n2, Alpha=@n4)) {group=Textures, parameterName=Lighting_Flip_1, texture=/DealershipSharedLibrary/FX/Textures/Lightning/T_Lightning_Flip_01}.R
   B=Constant() {r=0.5}
  )
  n6 = Add(A=NamedRerouteUsage(), B=TextureCoordinate() {vTiling=-1})
  n7 = Rotator(Coordinate=@n6, Time=Constant() {r=0.5}) {speed=3.14}
  n8 = NamedRerouteUsage()
  n9 = Multiply(
   A=TextureSampleParameter2D(UVs=LinearInterpolate(A=@n7, B=@n6, Alpha=@n8)) {group=Textures, parameterName=Lighting_Mirror, texture=&T_Lightning_Flip_01}.R
   B=Constant() {r=13}
  )
  NamedRerouteDeclaration(None=Multiply(A=NamedRerouteUsage(), B=LinearInterpolate(A=@n5, B=@n9, Alpha=@n1))) {name=Texture_Out, nodeColor=<rgba 0 1 0.2246 1>}
  n10 = DynamicParameter() {paramNames [DetailMotion, Brightness, RotateTexture, TextureSwitch]}
  NamedRerouteDeclaration(None=@n10.DetailMotion) {name=Detial Motion, nodeColor=<rgba 1 0.0076 0.0351 1>}
  n11 = Multiply(A=ParticleColor().RGB, B=NamedRerouteUsage())
  n12 = NamedRerouteUsage()
  NamedRerouteDeclaration(None=Multiply(A=ScalarParameter() {defaultValue=150, parameterName=Additional_Emissive}, B=Multiply(A=@n11, B=@n12))) {name=Emissive Out, nodeColor=<rgba 1 0.312 0.8314 1>}
  NamedRerouteDeclaration(None=@n10.Brightness) {name=Brightness, nodeColor=<rgba 0.0715 1 0 1>}
  NamedRerouteDeclaration(None=@n10.RotateTexture) {name=RotateTexture, nodeColor=<rgba 0.0288 0 1 1>}
  NamedRerouteDeclaration(None=@n10.TextureSwitch) {name=TextureSwitch, nodeColor=<rgba 1 0.958 0.9504 1>}
  n13 = TextureCoordinate()
  n14 = ScalarParameter() {defaultValue=0.25, group=NoiseMask, parameterName=Noise_Tile}
  n15 = Time()
  n16 = ScalarParameter() {defaultValue=0.5, group=NoiseMask, parameterName=Noise_Speed}
  n17 = TextureSampleParameter2D(UVs=Rotator(Coordinate=Multiply(A=@n13, B=@n14), Time=Multiply(A=@n15, B=@n16))) {
   group=NoiseMask
   parameterName=Normal_Noise
   samplerType=SAMPLERTYPE_Normal
   texture=/Game/Effects/Fort_Effects/Creatures/Elemental/Ice/TilingCloudNormalsMap
  }
  n18 = ComponentMask(None=@n17.RGB) {g=true, r=true}
  NamedRerouteDeclaration(None=Multiply(A=@n18, B=ScalarParameter() {defaultValue=0.025, group=NoiseMask, parameterName=Distorion_2})) {name=Distortion Opacity, nodeColor=<rgba 1 0.9301 0.1878 1>}
  n19 = ScalarParameter() {defaultValue=0.05, group=NoiseMask, parameterName=Distorion_1}
  NamedRerouteDeclaration(None=Add(A=TextureCoordinate() {uTiling=2.5, vTiling=2.5}, B=Multiply(A=@n18, B=@n19))) {name=Distortion Mask, nodeColor=<rgba 1 0.9301 0.1878 1>}
  n20 = TextureCoordinate() {uTiling=0.25, vTiling=0.25}
  n21 = LinearInterpolate(A=Panner(Coordinate=@n20) {speedX=0.25, speedY=-0.25}, B=Panner(Coordinate=@n20) {speedX=0.25, speedY=-0.15}, Alpha=NamedRerouteUsage())
  n22 = TextureSampleParameter2D(UVs=Panner(Coordinate=NamedRerouteUsage(), Speed=VectorParameter() {parameterName=Mask_Noise_Speed, defaultValue=<rgba 0.1 0.5 0 0>}.RGB)) {
   parameterName=Mask_Noise
   samplerType=SAMPLERTYPE_LinearGrayscale
   texture=/Game/Effects/Fort_Effects/Emotes/UltraLight/Texture/T_Plasma_A_Greyscale
  }
  n23 = Add(A=@n22.R, B=ScalarParameter() {defaultValue=0.1, group=NoiseMask, parameterName=Noise_Emis})
  NamedRerouteDeclaration(None=Multiply(
    A=TextureSampleParameter2D(UVs=@n21) {group=Textures, parameterName=Mask_Texture, texture=/DealershipSharedLibrary/FX/Textures/Masks/T_Grid}.R
    B=Clamp(None=@n23)
   )) {name=Mask Edge Texture, nodeColor=<rgba 0.4087 0.3059 1 1>}
  n24 = ParticleColor()
  n25 = ComponentMask(None=TextureCoordinate()) {g=true}
  n26 = ScalarParameter() {defaultValue=3, group=Alpha, parameterName=Alpha_Bottom_Gradient}
  n27 = Power(
   Base=ComponentMask(None=TextureCoordinate()) {g=true}
   Exp=ScalarParameter() {defaultValue=0.1, group=Alpha, parameterName=Alpha_Top_Gradient}
  )
  n28 = Multiply(A=NamedRerouteUsage(), B=Multiply(A=Power(Base=@n25, Exp=@n26), B=OneMinus(None=@n27)))
  NamedRerouteDeclaration(None=DepthFade(Opacity=Multiply(A=@n24.A, B=@n28), FadeDistance=ScalarParameter() {defaultValue=100, group=Alpha, parameterName=FadeDistance})) {name=Opacity Out, nodeColor=<rgba 0 0.954 1 1>}
  NamedRerouteUsage()
 ]
}
